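# UI material for the Snowfall item shop background: a blue gradient with a tinted dragon texture that shimmers over it.
# An animated mask adds panning snow dots, faded along V, to a panning dragon-movement texture with power-warped UVs, then clamps by ShimmerMax.
# That mask reveals the dragon, which is added to the gradient.
# Shared center-scaled UVs divide pixel height by itself, so their scale stays fixed rather than following aspect ratio.
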Material /ItemShop/UI/Art/Snowfall/M_UI_Snowfall_ShopBG
graph {
 properties {bUsedWithStaticMesh=false, materialDomain=MD_UI, pixelDepthOffsetMode=PDOM_Legacy, usedShadingModels=Default Lit}
 nodes [
  n1 = LinearInterpolate(
   A=Constant3Vector() {constant=<rgba 0 0.3231 1 0>}
   B=Constant3Vector() {constant=<rgba 0.0006 0.0262 0.4287 0>}
   Alpha=ComponentMask(None=NamedRerouteUsage()) {g=true}
  )
  n2 = Add(
   A=@n1
   B=Multiply(
    A=VectorParameter() {parameterName=Color, defaultValue=<rgba 0.0149 0.8337 0.9531 5.6892>}.RGB
    B=ScalarParameter() {defaultValue=0.25, parameterName=ColorSuppression}
   )
  )
  n3 = TextureSample() {texture=/ItemShop/UI/Art/Snowfall/Textures/T_UI_Dragon}
  n4 = TextureSample(UVs=Multiply(A=Panner() {speedX=0.008, speedY=-0.024}, B=Constant2Vector() {g=2, r=3}) {constA=2}) {texture=/Game/UI/Frontend/ItemShop/Art/Snowfall/Textures/T_UI_SnowfallDots}
  n5 = ComponentMask(None=TextureCoordinate() {uTiling=0.0782, vTiling=0.7839}) {g=true}
  n6 = Time()
  n7 = ScalarParameter() {defaultValue=0.6618, parameterName=Time factor}
  n8 = Power(
   Base=Panner(Coordinate=NamedRerouteUsage(), Time=Multiply(A=@n6, B=@n7)) {speedX=0.058, speedY=0.096}
   Exp=ScalarParameter() {defaultValue=0.9081, parameterName=Param_1}
  )
  n9 = Clamp(
   None=Add(
    A=LinearInterpolate(B=@n4.RGB, Alpha=@n5)
    B=TextureSample(UVs=@n8) {texture=/ItemShop/UI/Art/Snowfall/Textures/T_UI_SnowfallDragonMovement}.RGB
   )
   Max=ScalarParameter() {defaultValue=0.68, parameterName=ShimmerMax}
  )
  Add(A=LinearInterpolate(B=Multiply(A=@n2, B=@n3.R), Alpha=Saturate(None=@n9)), B=@n1)
  Multiply()
  n10 = MaterialFunctionCall() {materialFunction=&GetUserInterfaceUV}
  n11 = MaterialFunctionCall(Float2=@n10.Pixel Size) {materialFunction=&BreakOutFloat2Components}
  n12 = Multiply(A=Divide(A=@n11.G, B=@n11.G)) {constB=2}
  n13 = Constant() {r=1}
  NamedRerouteDeclaration(None=MaterialFunctionCall(UVs=@n10.Normalized UV, Texture Scale=AppendVector(A=@n12, B=@n13)) {materialFunction=&ScaleUVsByCenter}.UVs) {name=UVCoordinates, nodeColor=<rgba 1 0 0.8203 1>}
 ]
}
